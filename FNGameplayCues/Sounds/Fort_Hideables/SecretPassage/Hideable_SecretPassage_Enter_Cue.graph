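# Sound cue that plays the entry sound for the secret passage hideable.
# A single wave player for the secret passage enter sound feeds a fixed quarter-second delay at the root.
# It uses the traps sound class and attenuation, primes on load and plays at boosted volume.
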SoundCue /FNGameplayCues/Sounds/Fort_Hideables/SecretPassage/Hideable_SecretPassage_Enter_Cue
root=Delay_0
nodes [
 {id=Delay_0, children [WavePlayer_0], properties {delayMax=0.25, delayMin=0.25}}
 {id=WavePlayer_0, properties {soundWaveAssetPtr=&/Secret_Passage_Enter_V2}}
]
properties {
 attenuationSettings=&Fort_Traps_Att
 bPrimeOnLoad=true
 maxDistance=6000
 soundClassObject=&Fort_Traps
 volumeMultiplier=2
}
