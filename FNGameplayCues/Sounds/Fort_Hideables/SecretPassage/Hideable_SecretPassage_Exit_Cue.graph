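# Sound cue that plays the audio for exiting a Secret Passage hideable.
# A single wave player feeds the Secret_Passage_Exit_V2 wave straight to the output, with no randomization or modulation.
# It uses the Fort_Traps sound class and attenuation, boosts volume, and primes on load.
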SoundCue /FNGameplayCues/Sounds/Fort_Hideables/SecretPassage/Hideable_SecretPassage_Exit_Cue
root=WavePlayer_0
nodes [{id=WavePlayer_0, properties {soundWaveAssetPtr=&/Secret_Passage_Exit_V2}}]
properties {
 attenuationSettings=&Fort_Traps_Att
 bPrimeOnLoad=true
 maxDistance=6000
 soundClassObject=&Fort_Traps
 volumeMultiplier=1.75
}
