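# First-person sound cue for travelling through a secret passage hideable.
# A mixer layers a one-shot start wave, slightly boosted, with a looping travel wave.
# The loop passes through a fixed half-second delay and an enveloper volume ramp, so it fades in behind the start sound.
# The cue primes on load and uses the traps sound class and attenuation.
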SoundCue /FNGameplayCues/Sounds/Fort_Hideables/SecretPassage/SecretPassage_Travel_1P_Cue
root=Mixer_0
nodes [
 {id=Delay_0, children [WavePlayer_2], properties {delayMax=0.5, delayMin=0.5}}
 {
  id=Enveloper_0
  children [Delay_0]
  properties {pitchCurve {editorCurveData {keys [{time=0}]}}, volumeCurve {editorCurveData {keys [<at 0 0>,<at 0.25 0.0007>,<at 1 1>]}}}
 }
 {id=Mixer_0, children [WavePlayer_1,Enveloper_0], properties {inputVolume [1.15,1]}}
 {id=WavePlayer_1, properties {soundWaveAssetPtr=&/Secret_Passage_Travel_V2_Start}}
 {
  id=WavePlayer_2
  properties {bLooping=true, soundWaveAssetPtr=&/Secret_Passage_Travel_V2_Loop}
 }
]
properties {
 attenuationSettings=&Fort_Traps_Att
 bPrimeOnLoad=true
 maxDistance=6000
 soundClassObject=&Fort_Traps
 volumeMultiplier=1.6258
}
